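# MetaSound patch that applies a simple first-person suppressor effect to 5.1 weapon fire audio.
# Each gunshot triggers an AD envelope that shapes incoming front and center channels, and a briefly delayed fire subgraph playing suppressor sounds.
# The front pair mixes both sources, center is attenuated, and LFE and surrounds are muted.
# bEnable, checked per shot, routes either this processed result or the untouched input to the outputs.
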MetaSoundPatch /FortMetaSoundWeapon/Items/Metasounds/Attachments/MSP_Weapon_Fire_Suppressor_Simple_1p
inputs [
 BaseMSSAttackTime {type=Time}
 BaseMSSDecayTime {type=Time}
 BaseMSSVolumeMultiplier {type=Float}
 bEnable {type=Bool}
 bShouldHardReset {type=Bool}
 OnGunshot {type=Trigger}
 OutputVolumeMultiply {type=Float}
 Sounds.Close {type=WaveAsset:Array}
 Sounds.MaxPitch {type=Float}
 Sounds.MaxVolume {type=Float}
 Sounds.MinPitch {type=Float}
 Sounds.MinVolume {type=Float}
 Sounds.Volume {type=Float}
 "UE.InputFormat.5dot1.Audio:0" {type=Audio}
 "UE.InputFormat.5dot1.Audio:1" {type=Audio}
 "UE.InputFormat.5dot1.Audio:2" {type=Audio}
 "UE.InputFormat.5dot1.Audio:3" {type=Audio}
 "UE.InputFormat.5dot1.Audio:4" {type=Audio}
 "UE.InputFormat.5dot1.Audio:5" {type=Audio}
]
interfaces [MetaSound:1.0, UE.InputFormat.5dot1:1.0, UE.OutputFormat.5dot1:1.0]
outputs [
 "UE.OutputFormat.5dot1.Audio:0" {type=Audio}
 "UE.OutputFormat.5dot1.Audio:1" {type=Audio}
 "UE.OutputFormat.5dot1.Audio:2" {type=Audio}
 "UE.OutputFormat.5dot1.Audio:3" {type=Audio}
 "UE.OutputFormat.5dot1.Audio:4" {type=Audio}
 "UE.OutputFormat.5dot1.Audio:5" {type=Audio}
]
pages [Default {
  input_defaults {
   BaseMSSAttackTime=0.01
   BaseMSSDecayTime=0.15
   BaseMSSVolumeMultiplier=0.3
   OnGunshot=null
   OutputVolumeMultiply=1
   Sounds.MaxPitch=1.05
   Sounds.MaxVolume=1.05
   Sounds.MinPitch=0.95
   Sounds.MinVolume=0.95
   Sounds.Volume=1
   "UE.InputFormat.5dot1.Audio:0"=null
   "UE.InputFormat.5dot1.Audio:1"=null
   "UE.InputFormat.5dot1.Audio:2"=null
   "UE.InputFormat.5dot1.Audio:3"=null
   "UE.InputFormat.5dot1.Audio:4"=null
   "UE.InputFormat.5dot1.Audio:5"=null
   bEnable="<MetaSoundBuilderResult.SUCCEEDED: 0>"
   bShouldHardReset="<MetaSoundBuilderResult.SUCCEEDED: 0>"
   Sounds.Close []
  }
  nodes [
   TriggerDelay_0 = UE.Trigger Delay(In=^OnGunshot) {also {defaults {Delay Time=0.01}}}
   n1 = /FortMetaSoundWeapon/Items/Metasounds/Legacy/Subgraphs/MSS_Weapon_Fire_Main_5_1(
    FireSounds=^Sounds.Close
    MaxPitch=^Sounds.MaxPitch
    MaxVolume=^Sounds.MaxVolume
    MinPitch=^Sounds.MinPitch
    MinVolume=^Sounds.MinVolume
    UE.Source.OnPlay=@TriggerDelay_0.Out
   )
   ADEnvelopeADEnvelope_0 = AD Envelope.AD Envelope:Float(Attack Time=^BaseMSSAttackTime, Decay Time=^BaseMSSDecayTime, Trigger=^OnGunshot) {also {defaults {Attack Curve=2, Decay Curve=0.5}}}
   AudioMixerAudioMixerStereo2_0 = "AudioMixer.Audio Mixer (Stereo, 2)"(
    Gain 0=^BaseMSSVolumeMultiplier
    Gain 1=^Sounds.Volume
    In 1 L=@n1."UE.OutputFormat.5dot1.Audio:0"
    In 1 R=@n1."UE.OutputFormat.5dot1.Audio:1"
    In 0 R=UE.Multiply:Audio by Float(PrimaryOperand=^"UE.InputFormat.5dot1.Audio:1", AdditionalOperands=@ADEnvelopeADEnvelope_0.Out Envelope).Out
    In 0 L=UE.Multiply:Audio by Float(PrimaryOperand=^"UE.InputFormat.5dot1.Audio:0", AdditionalOperands=@ADEnvelopeADEnvelope_0.Out Envelope).Out
   )
   n2 = TriggerCompare.Clamp:Bool(Compare=^OnGunshot, A=^bEnable) {also {defaults {B="<MetaSoundBuilderResult.SUCCEEDED: 0>"}}}
   n3 = InitVariable.Audio()
   TriggerRouteTriggerRouteAudio2_0 = "TriggerRoute.Trigger Route (Audio, 2):Audio"(
    Value 0=UE.Multiply:Audio by Float(AdditionalOperands=^OutputVolumeMultiply, PrimaryOperand=@AudioMixerAudioMixerStereo2_0.Out R).Out
    Set 1=@n2.False
    Set 0=@n2.True
    Value 1=VariableAccessor.Audio(Variable=VariableMutator.Audio(Value=^"UE.InputFormat.5dot1.Audio:1", Variable=@n3.Variable).Variable).Value
   )
   n4 = InitVariable.Audio()
   TriggerRouteTriggerRouteAudio2_1 = "TriggerRoute.Trigger Route (Audio, 2):Audio"(
    Value 0=UE.Multiply:Audio by Float(PrimaryOperand=^"UE.InputFormat.5dot1.Audio:5") {also {defaults {AdditionalOperands=0}}}.Out
    Set 1=@n2.False
    Set 0=@n2.True
    Value 1=VariableAccessor.Audio(Variable=VariableMutator.Audio(Value=^"UE.InputFormat.5dot1.Audio:5", Variable=@n4.Variable).Variable).Value
   )
   n5 = InitVariable.Audio()
   TriggerRouteTriggerRouteAudio2_2 = "TriggerRoute.Trigger Route (Audio, 2):Audio"(
    Value 0=UE.Multiply:Audio by Float(PrimaryOperand=^"UE.InputFormat.5dot1.Audio:3") {also {defaults {AdditionalOperands=0}}}.Out
    Set 1=@n2.False
    Set 0=@n2.True
    Value 1=VariableAccessor.Audio(Variable=VariableMutator.Audio(Value=^"UE.InputFormat.5dot1.Audio:3", Variable=@n5.Variable).Variable).Value
   )
   n6 = InitVariable.Audio()
   TriggerRouteTriggerRouteAudio2_3 = "TriggerRoute.Trigger Route (Audio, 2):Audio"(
    Value 0=UE.Multiply:Audio by Float(AdditionalOperands=^OutputVolumeMultiply, PrimaryOperand=@AudioMixerAudioMixerStereo2_0.Out L).Out
    Set 1=@n2.False
    Set 0=@n2.True
    Value 1=VariableAccessor.Audio(Variable=VariableMutator.Audio(Value=^"UE.InputFormat.5dot1.Audio:0", Variable=@n6.Variable).Variable).Value
   )
   n7 = UE.Multiply:Audio by Float(PrimaryOperand=^"UE.InputFormat.5dot1.Audio:2", AdditionalOperands=@ADEnvelopeADEnvelope_0.Out Envelope)
   n8 = InitVariable.Audio()
   TriggerRouteTriggerRouteAudio2_4 = "TriggerRoute.Trigger Route (Audio, 2):Audio"(
    Value 0=UE.Multiply:Audio by Float(
     AdditionalOperands=^OutputVolumeMultiply
     PrimaryOperand=UE.Multiply:Audio by Float(PrimaryOperand=@n7.Out) {also {defaults {AdditionalOperands=0.6}}}.Out
    ).Out
    Set 1=@n2.False
    Set 0=@n2.True
    Value 1=VariableAccessor.Audio(Variable=VariableMutator.Audio(Value=^"UE.InputFormat.5dot1.Audio:2", Variable=@n8.Variable).Variable).Value
   )
   n9 = InitVariable.Audio()
   TriggerRouteTriggerRouteAudio2_5 = "TriggerRoute.Trigger Route (Audio, 2):Audio"(
    Value 0=UE.Multiply:Audio by Float(PrimaryOperand=^"UE.InputFormat.5dot1.Audio:4") {also {defaults {AdditionalOperands=0}}}.Out
    Set 1=@n2.False
    Set 0=@n2.True
    Value 1=VariableAccessor.Audio(Variable=VariableMutator.Audio(Value=^"UE.InputFormat.5dot1.Audio:4", Variable=@n9.Variable).Variable).Value
   )
  ]
  links [
   {dst="@UE.OutputFormat.5dot1.Audio:1", src=TriggerRouteTriggerRouteAudio2_0.Value}
   {dst="@UE.OutputFormat.5dot1.Audio:5", src=TriggerRouteTriggerRouteAudio2_1.Value}
   {dst="@UE.OutputFormat.5dot1.Audio:3", src=TriggerRouteTriggerRouteAudio2_2.Value}
   {dst="@UE.OutputFormat.5dot1.Audio:0", src=TriggerRouteTriggerRouteAudio2_3.Value}
   {dst="@UE.OutputFormat.5dot1.Audio:2", src=TriggerRouteTriggerRouteAudio2_4.Value}
   {dst="@UE.OutputFormat.5dot1.Audio:4", src=TriggerRouteTriggerRouteAudio2_5.Value}
  ]
 }]
