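# MetaSound patch preset of MSP_Weapon_Fire_Suppressor_Simple_1p that supplies first-person suppressed sniper gunshot audio as a weapon attachment mod.
# It takes an OnGunshot trigger and 5.1 audio input, and outputs 5.1 audio through the parent's graph.
# The preset supplies five close sniper silencer waves with pitch and volume ranges.
# BaseMSS attack, decay and volume-multiplier inputs suggest the parent envelopes and attenuates the base weapon sound.
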MetaSoundPatch /FortMetaSoundWeapon/Items/Metasounds/Attachments/Presets/Preset_Mod_Suppressor_Simple_1p_Sniper
preset_parent=/FortMetaSoundWeapon/Items/Metasounds/Attachments/MSP_Weapon_Fire_Suppressor_Simple_1p
inputs [
 BaseMSSAttackTime {type=Time}
 BaseMSSDecayTime {type=Time}
 BaseMSSVolumeMultiplier {type=Float}
 bEnable {type=Bool}
 bShouldHardReset {type=Bool}
 OnGunshot {type=Trigger}
 OutputVolumeMultiply {type=Float}
 Sounds.Close {type=WaveAsset:Array}
 Sounds.MaxPitch {type=Float}
 Sounds.MaxVolume {type=Float}
 Sounds.MinPitch {type=Float}
 Sounds.MinVolume {type=Float}
 Sounds.Volume {type=Float}
 "UE.InputFormat.5dot1.Audio:0" {type=Audio}
 "UE.InputFormat.5dot1.Audio:1" {type=Audio}
 "UE.InputFormat.5dot1.Audio:2" {type=Audio}
 "UE.InputFormat.5dot1.Audio:3" {type=Audio}
 "UE.InputFormat.5dot1.Audio:4" {type=Audio}
 "UE.InputFormat.5dot1.Audio:5" {type=Audio}
]
interfaces [MetaSound:1.0, UE.InputFormat.5dot1:1.0, UE.OutputFormat.5dot1:1.0]
outputs [
 "UE.OutputFormat.5dot1.Audio:0" {type=Audio}
 "UE.OutputFormat.5dot1.Audio:1" {type=Audio}
 "UE.OutputFormat.5dot1.Audio:2" {type=Audio}
 "UE.OutputFormat.5dot1.Audio:3" {type=Audio}
 "UE.OutputFormat.5dot1.Audio:4" {type=Audio}
 "UE.OutputFormat.5dot1.Audio:5" {type=Audio}
]
pages [Default {input_defaults {
   BaseMSSAttackTime=0.01
   BaseMSSDecayTime=0.18
   BaseMSSVolumeMultiplier=0.3
   OnGunshot=null
   OutputVolumeMultiply=1
   Sounds.MaxPitch=1.05
   Sounds.MaxVolume=1.05
   Sounds.MinPitch=0.95
   Sounds.MinVolume=0.95
   Sounds.Volume=1
   "UE.InputFormat.5dot1.Audio:0"=null
   "UE.InputFormat.5dot1.Audio:1"=null
   "UE.InputFormat.5dot1.Audio:2"=null
   "UE.InputFormat.5dot1.Audio:3"=null
   "UE.InputFormat.5dot1.Audio:4"=null
   "UE.InputFormat.5dot1.Audio:5"=null
   bEnable="<MetaSoundBuilderResult.SUCCEEDED: 0>"
   bShouldHardReset="<MetaSoundBuilderResult.SUCCEEDED: 0>"
   Sounds.Close [
    /Game/Sounds/Fort_Weapons/Suppressor/Silencer_Sniper_Close_01
    /Game/Sounds/Fort_Weapons/Suppressor/Silencer_Sniper_Close_02
    /Game/Sounds/Fort_Weapons/Suppressor/Silencer_Sniper_Close_03
    /Game/Sounds/Fort_Weapons/Suppressor/Silencer_Sniper_Close_04
    /Game/Sounds/Fort_Weapons/Suppressor/Silencer_Sniper_Close_05
   ]
  }}]
