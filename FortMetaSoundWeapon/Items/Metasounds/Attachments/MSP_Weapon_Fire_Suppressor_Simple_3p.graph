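# MetaSound patch that applies a simple suppressor attachment to third-person weapon fire audio, processing the incoming stereo signal on each gunshot.
# An AD envelope shapes the input, which is scaled by a volume multiplier and the inverse distant curve.
# Random close stereo and medium mono one-shots, weighted by distance curves, join it in a three-input mixer.
# A gunshot-triggered bEnable compare routes either that mix or the unprocessed input to output.
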MetaSoundPatch /FortMetaSoundWeapon/Items/Metasounds/Attachments/MSP_Weapon_Fire_Suppressor_Simple_3p
inputs [
 BaseMSSAttackTime {type=Time}
 BaseMSSDecayTime {type=Time}
 BaseMSSVolumeMultiplier {type=Float}
 bEnable {type=Bool}
 bShouldHardReset {type=Bool}
 CloseCurveFloat {type=Float}
 DistantCurveFloat {type=Float}
 MediumCurveFloat {type=Float}
 OnGunshot {type=Trigger}
 OutputVolumeMultiply {type=Float}
 Sounds.Close {type=WaveAsset:Array}
 Sounds.CloseVolume {type=Float}
 Sounds.MaxPitch {type=Float}
 Sounds.MaxVolume {type=Float}
 Sounds.Medium {type=WaveAsset:Array}
 Sounds.MediumVolume {type=Float}
 Sounds.MinPitch {type=Float}
 Sounds.MinVolume {type=Float}
 "UE.InputFormat.Stereo.Audio:0" {type=Audio}
 "UE.InputFormat.Stereo.Audio:1" {type=Audio}
]
interfaces [MetaSound:1.0, UE.InputFormat.Stereo:1.0, UE.OutputFormat.Stereo:1.0]
outputs [
 "UE.OutputFormat.Stereo.Audio:0" {type=Audio}
 "UE.OutputFormat.Stereo.Audio:1" {type=Audio}
]
pages [Default {
  input_defaults {
   BaseMSSAttackTime=0.01
   BaseMSSDecayTime=0.08
   BaseMSSVolumeMultiplier=0.8
   CloseCurveFloat=0
   DistantCurveFloat=0
   MediumCurveFloat=0
   OnGunshot=null
   OutputVolumeMultiply=1
   Sounds.CloseVolume=1
   Sounds.MaxPitch=1.05
   Sounds.MaxVolume=1.05
   Sounds.MediumVolume=1
   Sounds.MinPitch=0.95
   Sounds.MinVolume=0.95
   "UE.InputFormat.Stereo.Audio:0"=null
   "UE.InputFormat.Stereo.Audio:1"=null
   bEnable=null
   bShouldHardReset="<MetaSoundBuilderResult.SUCCEEDED: 0>"
   Sounds.Close []
   Sounds.Medium []
  }
  nodes [
   n1 = &MSS_PlayRandomSound_Stereo(
    UE.Source.OnPlay=^OnGunshot
    Sounds=^Sounds.Close
    MaxPitch=^Sounds.MaxPitch
    MaxVolume=^Sounds.MaxVolume
    MinPitch=^Sounds.MinPitch
    MinVolume=^Sounds.MinVolume
   )
   n2 = UE.Reroute(Value=UE.Reroute(Value=^CloseCurveFloat).Value)
   ADEnvelopeADEnvelope_0 = AD Envelope.AD Envelope:Float(Attack Time=^BaseMSSAttackTime, Decay Time=^BaseMSSDecayTime, Trigger=^OnGunshot) {also {defaults {Attack Curve=2, Decay Curve=0.5, Hard Reset="<MetaSoundBuilderResult.SUCCEEDED: 0>"}}}
   n3 = UE.Multiply:Audio by Float(PrimaryOperand=^"UE.InputFormat.Stereo.Audio:0", AdditionalOperands=@ADEnvelopeADEnvelope_0.Out Envelope)
   n4 = UE.Subtract:Float(AdditionalOperands=^DistantCurveFloat) {also {defaults {PrimaryOperand=1}}}
   n5 = UE.Multiply:Audio by Float(PrimaryOperand=^"UE.InputFormat.Stereo.Audio:1", AdditionalOperands=@ADEnvelopeADEnvelope_0.Out Envelope)
   n6 = UE.Reroute(Value=^MediumCurveFloat)
   n7 = UE.Multiply:Audio by Float(
    AdditionalOperands=^Sounds.MediumVolume
    PrimaryOperand=UE.Multiply:Audio by Float(
     PrimaryOperand=&MSS_PlayRandomSound_Mono(
      UE.Source.OnPlay=^OnGunshot
      MaxPitch=^Sounds.MaxPitch
      MaxVolume=^Sounds.MaxVolume
      Sounds=^Sounds.Medium
      MinPitch=^Sounds.MinPitch
      MinVolume=^Sounds.MinVolume
     )."UE.OutputFormat.Mono.Audio:0"
     AdditionalOperands=UE.Reroute(Value=@n6.Value).Value
    ).Out
   )
   AudioMixerAudioMixerStereo3_0 = "AudioMixer.Audio Mixer (Stereo, 3)"(
    Gain 0=^OutputVolumeMultiply
    Gain 1=^OutputVolumeMultiply
    Gain 2=^OutputVolumeMultiply
    In 1 R=UE.Multiply:Audio by Float(
     AdditionalOperands=^Sounds.CloseVolume
     PrimaryOperand=UE.Multiply:Audio by Float(PrimaryOperand=@n1."UE.OutputFormat.Stereo.Audio:1", AdditionalOperands=@n2.Value).Out
    ).Out
    In 0 L=UE.Multiply:Audio by Float(
     PrimaryOperand=UE.Multiply:Audio by Float(AdditionalOperands=^BaseMSSVolumeMultiplier, PrimaryOperand=@n3.Out).Out
     AdditionalOperands=@n4.Out
    ).Out
    In 0 R=UE.Multiply:Audio by Float(
     PrimaryOperand=UE.Multiply:Audio by Float(AdditionalOperands=^BaseMSSVolumeMultiplier, PrimaryOperand=@n5.Out).Out
     AdditionalOperands=@n4.Out
    ).Out
    In 2 L=@n7.Out
    In 2 R=@n7.Out
    In 1 L=UE.Multiply:Audio by Float(
     AdditionalOperands=^Sounds.CloseVolume
     PrimaryOperand=UE.Multiply:Audio by Float(PrimaryOperand=@n1."UE.OutputFormat.Stereo.Audio:0", AdditionalOperands=@n2.Value).Out
    ).Out
   )
   n8 = TriggerCompare.Clamp:Bool(Compare=^OnGunshot, A=^bEnable) {also {defaults {B="<MetaSoundBuilderResult.SUCCEEDED: 0>"}}}
   n9 = InitVariable.Audio()
   TriggerRouteTriggerRouteAudio2_0 = "TriggerRoute.Trigger Route (Audio, 2):Audio"(
    Value 0=@AudioMixerAudioMixerStereo3_0.Out L
    Set 1=@n8.False
    Set 0=@n8.True
    Value 1=VariableAccessor.Audio(Variable=VariableMutator.Audio(Value=^"UE.InputFormat.Stereo.Audio:0", Variable=@n9.Variable).Variable).Value
   )
   n10 = InitVariable.Audio()
   TriggerRouteTriggerRouteAudio2_1 = "TriggerRoute.Trigger Route (Audio, 2):Audio"(
    Value 0=@AudioMixerAudioMixerStereo3_0.Out R
    Set 1=@n8.False
    Set 0=@n8.True
    Value 1=VariableAccessor.Audio(Variable=VariableMutator.Audio(Value=^"UE.InputFormat.Stereo.Audio:1", Variable=@n10.Variable).Variable).Value
   )
  ]
  links [
   {dst="@UE.OutputFormat.Stereo.Audio:0", src=TriggerRouteTriggerRouteAudio2_0.Value}
   {dst="@UE.OutputFormat.Stereo.Audio:1", src=TriggerRouteTriggerRouteAudio2_1.Value}
  ]
 }]
