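# MetaSound patch preset that configures the simple third-person weapon suppressor fire sound for shotguns.
# It inherits its graph from MSP_Weapon_Fire_Suppressor_Simple_3p and supplies input values, notably six close and six medium shotgun silencer wave variations.
# Inputs expose a gunshot trigger, close and medium curve values, and narrow pitch and volume ranges.
# The patch takes and returns stereo audio, with attack, decay and volume inputs for the base MetaSound signal.
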MetaSoundPatch /FortMetaSoundWeapon/Items/Metasounds/Attachments/Presets/Preset_Mod_Suppressor_Simple_3p_Shotgun
preset_parent=/FortMetaSoundWeapon/Items/Metasounds/Attachments/MSP_Weapon_Fire_Suppressor_Simple_3p
inputs [
 BaseMSSAttackTime {type=Time}
 BaseMSSDecayTime {type=Time}
 BaseMSSVolumeMultiplier {type=Float}
 bEnable {type=Bool}
 bShouldHardReset {type=Bool}
 CloseCurveFloat {type=Float}
 MediumCurveFloat {type=Float}
 OnGunshot {type=Trigger}
 OutputVolumeMultiply {type=Float}
 Sounds.Close {type=WaveAsset:Array}
 Sounds.CloseVolume {type=Float}
 Sounds.MaxPitch {type=Float}
 Sounds.MaxVolume {type=Float}
 Sounds.Medium {type=WaveAsset:Array}
 Sounds.MediumVolume {type=Float}
 Sounds.MinPitch {type=Float}
 Sounds.MinVolume {type=Float}
 "UE.InputFormat.Stereo.Audio:0" {type=Audio}
 "UE.InputFormat.Stereo.Audio:1" {type=Audio}
]
interfaces [MetaSound:1.0, UE.InputFormat.Stereo:1.0, UE.OutputFormat.Stereo:1.0]
outputs [
 "UE.OutputFormat.Stereo.Audio:0" {type=Audio}
 "UE.OutputFormat.Stereo.Audio:1" {type=Audio}
]
pages [Default {input_defaults {
   BaseMSSAttackTime=0.01
   BaseMSSDecayTime=0.12
   BaseMSSVolumeMultiplier=0.5
   CloseCurveFloat=0
   MediumCurveFloat=0
   OnGunshot=null
   OutputVolumeMultiply=0.5
   Sounds.CloseVolume=1
   Sounds.MaxPitch=1.05
   Sounds.MaxVolume=1.05
   Sounds.MediumVolume=1
   Sounds.MinPitch=0.95
   Sounds.MinVolume=0.95
   "UE.InputFormat.Stereo.Audio:0"=null
   "UE.InputFormat.Stereo.Audio:1"=null
   bEnable=null
   bShouldHardReset="<MetaSoundBuilderResult.SUCCEEDED: 0>"
   Sounds.Close [
    /Game/Sounds/Fort_Weapons/Suppressor/Silencer_Shotgun_Close_01
    /Game/Sounds/Fort_Weapons/Suppressor/Silencer_Shotgun_Close_02
    /Game/Sounds/Fort_Weapons/Suppressor/Silencer_Shotgun_Close_03
    /Game/Sounds/Fort_Weapons/Suppressor/Silencer_Shotgun_Close_04
    /Game/Sounds/Fort_Weapons/Suppressor/Silencer_Shotgun_Close_05
    /Game/Sounds/Fort_Weapons/Suppressor/Silencer_Shotgun_Close_06
   ]
   Sounds.Medium [
    /Game/Sounds/Fort_Weapons/Suppressor/Silencer_Shotgun_Medium_01
    /Game/Sounds/Fort_Weapons/Suppressor/Silencer_Shotgun_Medium_02
    /Game/Sounds/Fort_Weapons/Suppressor/Silencer_Shotgun_Medium_03
    /Game/Sounds/Fort_Weapons/Suppressor/Silencer_Shotgun_Medium_04
    /Game/Sounds/Fort_Weapons/Suppressor/Silencer_Shotgun_Medium_05
    /Game/Sounds/Fort_Weapons/Suppressor/Silencer_Shotgun_Medium_06
   ]
  }}]
